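; Skydive sound library component that drives skydiving audio from Glide, Freefall and Land gameplay tag events.
; Tracks played audio components, sends them matching trigger parameters and clears them on Land; stopped base Skydive sounds fade out logarithmically.
; Registers the audio analysis default submix on Freefall and unregisters it on Land.
; An event delegate activates the contrail control bus mix once after a short delay and deactivates it on Land.
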(event EventBeginPlay
  (|CallParentFunction)
  (EventDispatchers|AssignOnEventPlayedDelegate (delegate Custom|ActivateCBM)))

(event Custom|ActivateCBM
  (params (EventName :type Gameplay%20Tag%20Structure))
  (bind _contrail_cbm (Variables|Default|GetContrail%20CBM))
  (switch Utilities|FlowControl|Switch|SwitchonGameplayTag EventName
    (:SoundLibrary.Skydive.Glide
      _)
    (:SoundLibrary.Skydive.Freefall
      _)
    (:SoundLibrary.Skydive.Land
      (Audio|Modulation|DeactivateControlBusMix _contrail_cbm)))
  (Utilities|FlowControl|DoOnce false
    (:Completed
      (Utilities|FlowControl|Delay 1.0)
      (Audio|Modulation|ActivateControlBusMix _contrail_cbm))))

(fn OnEventPlayed
  (params (InEventName :type Gameplay%20Tag%20Structure))
  (bind _returnvalue (WorldSubsystems|GetFortAudioAnalysisSubsystem))
  (bind _components (Variables|Default|GetComponents))
  (switch Utilities|FlowControl|Switch|SwitchonGameplayTag InEventName
    (:SoundLibrary.Skydive.Glide
      (Audio|Parameter|ExecuteTriggerParameter _components "OnGlide"))
    (:SoundLibrary.Skydive.Freefall
      (Audio|Parameter|ExecuteTriggerParameter _components "OnFreefall")
      (if (not (Variables|Default|GetbIsSubmixRegistered))
        (AudioAnalysis|RegisterDefaultSubmix _returnvalue)
        (Variables|Default|SetbIsSubmixRegistered true)))
    (:SoundLibrary.Skydive.Land
      (Audio|Parameter|ExecuteTriggerParameter _components "OnLand")
      (Utilities|Array|Clear (Variables|Default|GetComponents))
      (if (Variables|Default|GetbIsSubmixRegistered)
        (AudioAnalysis|UnRegisterDefaultSubmix _returnvalue)
        (Variables|Default|SetbIsSubmixRegistered false)))))

(fn OnSoundPlayed
  (params (InEventName :type Gameplay%20Tag%20Structure) (InComponent :type Audio%20Component%20Object%20Reference))
  (Utilities|Array|Add (Variables|Default|GetComponents) InComponent))

(fn OnSoundStopped
  (params (InEventName :type Gameplay%20Tag%20Structure) (InComponent :type Audio%20Component%20Object%20Reference))
  (bind _returnvalue (== InEventName "(TagName=\"SoundLibrary.Skydive\")"))
  (if _returnvalue
    (Audio|Components|Audio|FadeOut InComponent 3.5 0.0 "Logarithmic" :!class /Script/Engine.AudioComponent)
    (return _returnvalue)))
